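# Default Lit surface material for the PinHammer accessory prototype in the Fortress weapons set.
# It feeds the complete map as diffuse, and the specular map scaled by a small base Fresnel reflectance factor, into ConvertFromDiffSpec.
# That function adapts diffuse/specular-authored textures to Unreal's PBR inputs.
# A separate normal map sample supplies surface detail.
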Material /Game/Weapons/Fortress_Weapons/Materials/M_AccessoryProto_PinHammer_Mat
graph {
 properties {pixelDepthOffsetMode=PDOM_Legacy, usedShadingModels=Default Lit}
 nodes [
  MaterialFunctionCall(
   DiffuseColor=TextureSample() {texture=&/SK_AccessoryProto_PinHammerCompleteMap}.RGB
   SpecularColor=Multiply(A=TextureSample() {texture=&/SK_AccessoryProto_PinHammerSpecularMap}.RGB) {constA=1, constB=0.04, desc=FresnelBaseReflectFraction}
  ) {materialFunction=&ConvertFromDiffSpec}
  TextureSample() {samplerType=SAMPLERTYPE_Normal, texture=&/SK_AccessoryProto_PinHammerNormalsMap}
 ]
}
